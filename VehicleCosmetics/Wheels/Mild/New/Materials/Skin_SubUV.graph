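# Material function for animated flipbook textures in wheel cosmetic skins, outputting RGB and alpha from a sub-image grid.
# Time scaled by Speed is wrapped, shifted by a frame-position offset and multiplied by the grid's cell count to drive SubUVCoord.
# A static switch selects either the current frame or a linear blend with the next.
# The blend weight is the fraction of Speed itself, so it does not follow frame progress.
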MaterialFunction /VehicleCosmetics/Wheels/Mild/New/Materials/Skin_SubUV
graph {nodes [
  n1 = FunctionInput() {bUsePreviewValueAsDefault=true, inputName=SubImages, inputType=FunctionInput_Vector2, sortPriority=1, previewValue=<1 1 0 1>}
  n2 = FunctionInput(Preview=TextureCoordinate()) {inputName=UVs, inputType=FunctionInput_Vector2, sortPriority=3}
  n3 = FunctionInput() {bUsePreviewValueAsDefault=true, inputName=Speed, inputType=FunctionInput_Scalar, sortPriority=2}
  n4 = Multiply(A=Time(), B=@n3)
  n5 = Multiply(
   A=Add(
    A=FunctionInput() {bUsePreviewValueAsDefault=true, inputName=Offset_Frame_Position, inputType=FunctionInput_Scalar, sortPriority=4}
    B=Frac(None=@n4)
   )
   B=Multiply(A=ComponentMask(None=@n1) {r=true}, B=ComponentMask(None=@n1) {g=true})
  )
  n6 = MaterialFunctionCall!(SubImages=@n1, UVs=@n2, Frame=@n5) {materialFunction=&/SubUVCoord}
  n7 = FunctionInput(Preview=TextureObject()) {bUsePreviewValueAsDefault=true, inputName=Texture, inputType=FunctionInput_Texture2D}
  n8 = TextureSample(UVs=@n6.Result, Tex=@n7) {samplerType=SAMPLERTYPE_LinearColor}
  n9 = Constant() {r=1}
  n10 = TextureSample(
   UVs=MaterialFunctionCall(SubImages=@n1, UVs=@n2, Frame=Add(A=@n5, B=@n9)) {materialFunction=&SubUVCoord}.Result
   Tex=@n7
  )
  n11 = Frac(None=@n3)
  n12 = TextureSample(UVs=@n6.Result, Tex=@n7)
  n13 = FunctionInput(Preview=StaticBool()) {inputName=Linear_Blend, inputType=FunctionInput_StaticBool, sortPriority=5}
  FunctionOutput(None=StaticSwitch(True=LinearInterpolate(A=@n8.RGB, B=@n10.RGB, Alpha=@n11), False=@n12.RGB, Value=@n13)) {outputName=Output_RGB}
  FunctionOutput(None=StaticSwitch(True=LinearInterpolate(A=@n8.A, B=@n10.A, Alpha=@n11), False=@n12.A, Value=@n13)) {outputName=Output_Alpha}
 ]}
